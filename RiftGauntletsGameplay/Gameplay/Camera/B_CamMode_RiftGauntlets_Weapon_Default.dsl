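; Default weapon camera mode for Rift Gauntlets that keeps the camera from being blocked by objects the player dashes through.
; It overrides GetAdditionalPenetrationIgnoreActors to supply extra actors for camera penetration checks to skip.
; For a locally controlled pawn with a Dash or Ascend gameplay tag, it reuses the capsule component's move-ignore actor list.
; Otherwise, it adds no ignored actors.
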(fn GetAdditionalPenetrationIgnoreActors
  (params (ViewTarget :type Actor%20Object%20Reference) (CameraLoc :type Vector%20%28by%20ref%29 :default "0, 0, 0") (TargetLoc :type Vector%20%28by%20ref%29 :default "0, 0, 0"))
  ;[0.06237 0.15 0.1258 0.5] When dashing, we move through certain smaller objects. Camera shouldnt be blocked by those objects.
  (Utilities|FlowControl|Sequence
    (:then_0
      (bind _asfort_player_pawn (Utilities|Casting|CastToFortPlayerPawn ViewTarget))
      (if (Pawn|IsLocallyControlled _asfort_player_pawn)
        (if (GameplayTags|HasAnyMatchingGameplayTags _asfort_player_pawn (GameplayTags|MakeLiteralGameplayTagContainer "(GameplayTags=((TagName=\"RiftGauntlets.Ability.Dash\"),(TagName=\"RiftGauntlets.Ability.Ascend\")))"))
          (bind _capsulecomponent (Variables|/Script/Engine.Character|GetCapsuleComponent _asfort_player_pawn)
            (:then
              (bind _returnvalue (Collision|GetMoveIgnoreActors _capsulecomponent))
              (return _returnvalue))
            (:else)))))
    (:then_1
      (return 0))))
