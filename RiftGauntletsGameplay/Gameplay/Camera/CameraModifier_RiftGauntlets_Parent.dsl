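; Parent camera modifier for the Rift Gauntlets that adjusts FOV, position and lag for hover and dash states.
; Unless native calculation is used, it chains an interpolated view-relative offset into a lag pulling against view-target velocity, both dash-only, and leaves rotation unchanged.
; FOV interpolates toward hover or dash targets, and all effects scale by modifier alpha.
; NotifyDash sets Blueprint and native dash flags; DisableAndRemove waits out the alpha fade before removal.
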(event Custom|DisableAndRemove
  (CameraModifier|DisableModifier false)
  (Utilities|FlowControl|Delay (Variables|CameraModifier|GetAlphaOutTime))
  (Game|Player|RemoveCameraModifier (Variables|CameraModifier|GetCameraOwner) self))

(fn BlueprintModifyCamera
  (params (DeltaTime :type Float%20%28single-precision%29) (ViewLocation :type Vector :default "0, 0, 0") (ViewRotation :type Rotator :default "0, 0, 0") (FOV :type Float%20%28single-precision%29))
  (if (CallFunction|UseNativeCalculation)
    (return (Variables|GetViewLocation) (Variables|GetViewRotation) (Variables|GetFOV))
    (else
      (bind _modifiedcameralocation (Default|ModifyCameraOffset ViewLocation ViewRotation DeltaTime))
      (bind _resultlocation (Default|ModifyCameraLagLocation _modifiedcameralocation DeltaTime))
      (bind _outfov (CallFunction|ModifyCameraFieldOfView FOV DeltaTime))
      (return _resultlocation (Variables|GetViewRotation) _outfov))))

(fn ModifyCameraFieldOfView
  (params (InFov :type Float%20%28double-precision%29) (DeltaTime :type Float%20%28double-precision%29))
  (Utilities|FlowControl|Sequence
    (:then_0
      (if (Variables|Settings|GetbModifyFov)
        (Variables|Transient|SetModifyFOV (Math|Interpolation|FInterpTo (Variables|Transient|GetModifyFOV) (Math|Float|SelectFloat (Variables|Settings|GetTargetFOV_Dash) (Variables|Settings|GetTargetFOV_Hover) (Variables|Transient|GetbDash)) DeltaTime (Math|Float|SelectFloat (Variables|Settings|GetInterpSpeedFOV_Dash) (Variables|Settings|GetInterpSpeedFOV_Hover) (Variables|Transient|GetbDash))))))
    (:then_1
      (return (+ InFov (* (Variables|Transient|GetModifyFOV) (Variables|CameraModifier|GetAlpha)))))))

(fn ModifyCameraLagLocation
  (params (InLocation :type Vector) (DeltaTime :type Float%20%28double-precision%29))
  (Utilities|FlowControl|Sequence
    (:then_0
      (LocalVariables|SetReturnLocation InLocation)
      (if (and (Variables|Transient|GetbDash) (Variables|Settings|GetbModifyCameraLag))
        (bind _cameraowner (Variables|CameraModifier|GetCameraOwner)
          (:then
            (Utilities|IsValid (Game|Player|GetOwningPlayerController _cameraowner)
              (:Is%20Valid
                (Utilities|IsValid (Pawn|GetViewTarget (Game|Player|GetOwningPlayerController _cameraowner))
                  (:Is%20Valid
                    (Variables|Transient|SetCameraLagDelta (Math|Interpolation|VInterpTo (Variables|Transient|GetCameraLagDelta) (Utilities|Operators|Multiply (Math|Vector|Normalize (Transformation|GetVelocity (Pawn|GetViewTarget (Game|Player|GetOwningPlayerController _cameraowner)))) (Variables|Settings|GetCameraLagMagnitude) -1.0) (Variables|GetDeltaTime) (Variables|Settings|GetCameraLagInterpSeed)))
                    (return (+ (Variables|GetInLocation) (* (Variables|Transient|GetCameraLagDelta) (Variables|CameraModifier|GetAlpha)))))))))
          (:else))))
    (:then_1
      (return (LocalVariables|GetReturnLocation)))))

(fn ModifyCameraOffset
  (params (InCameraLocation :type Vector) (InCameraRotation :type Rotator) (DeltaTime :type Float%20%28double-precision%29))
  (bind (_cameraoffsettarget_x _cameraoffsettarget_y _cameraoffsettarget_z) (Math|Vector|BreakVector (Variables|GetCameraOffsetTarget_X)))
  (Utilities|FlowControl|Sequence
    (:then_0
      (if (and (Variables|Settings|GetbModifyCameraOffset) (Variables|Transient|GetbDash))
        (Variables|Transient|SetCameraOffsetDelta (Math|Interpolation|VInterpTo (Variables|Transient|GetCameraOffsetDelta) (Utilities|Operators|Add (* (Math|Vector|GetForwardVector InCameraRotation) _cameraoffsettarget_x) (* (Math|Vector|GetRightVector InCameraRotation) _cameraoffsettarget_y) (* (Math|Vector|GetUpVector InCameraRotation) _cameraoffsettarget_z)) DeltaTime (Variables|Settings|GetCameraOffsetTargetInterpSpeed)))))
    (:then_1
      (return (+ InCameraLocation (* (Variables|Transient|GetCameraOffsetDelta) (Variables|CameraModifier|GetAlpha)))))))

(fn NotifyDash
  (Variables|Transient|SetbDash true)
  (Variables|FortCameraModifier_RiftGauntlets|SetbNativeDash true))
